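# Physics asset for the F_MED_Dieselpunk_02 hat accessory, giving its dyn_bun bone a simulation setup relative to the head.
# A capsule body covers dyn_bun and a sphere body covers the head bone.
# A single constraint joins the two with tight swing and twist limits, restricting the bun to slight movement.
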PhysicsAsset /Game/Accessories/Hats/F_MED_Dieselpunk_02/Meshes/F_MED_Dieselpunk_02_Physics
bodies {
 dyn_bun [{shapeType=Capsule, length=4.4952, radius=1.744, center=<-0.3488 -4.0694 0>, rotation=<rot 0 -3.1279 90>}]
 head [{shapeType=Sphere, radius=9.5164, center=<5.6806 1.6978 0.2721>}]
}
constraints [{bone1Name=dyn_bun, bone2Name=head, swing1LimitDegrees=6, swing2LimitDegrees=6, twistLimitDegrees=2}]
