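# Post-style animation blueprint for the hair mesh of a Dieselpunk hat accessory, adding physics-driven secondary motion.
# It copies the pose from the attached parent mesh, then runs a RigidBody node with an override physics asset, simulated in spine_05 bone space with amplified component motion.
# An Anim Dynamics node for the dyn_bun bone is present but not connected to the output pose.
# The event graph holds no logic.
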AnimBlueprint /Game/Accessories/Hats/F_MED_Dieselpunk_02/Meshes/F_MED_Dieselpunk_02_05_hair_MESH_Skeleton_AnimBlueprint
target_skeleton=&/F_MED_Dieselpunk_02_05_hair_MESH_Skeleton
graphs [
 AnimGraph {nodes [
   Anim Dynamics - Bone: dyn_bun(Alpha=0.75, ExternalForce="0,0,0") {
    alpha=0.9
    angularDampingOverride=0.97
    angularSpringConstant=600
    bAngularSpring=true
    bGravityOverrideInSimSpace=true
    bOverrideAngularDamping=true
    gravityScale=0.2
    numSolverIterationsPostUpdate=4
    numSolverIterationsPreUpdate=16
    simulationSpace=Actor
    alphaBoolBlend {blendOption=Linear}
    boundBone {boneName=dyn_bun}
    componentAppliedLinearAccClamp=<1000 1000 1000>
    componentLinearAccScale=<1 1 1>
    componentLinearVelScale=<0.5 0.5 0.5>
    physicsBodyDefinitions [{
      sphereCollisionRadius=0
      boundBone {boneName=dyn_bun}
      boxExtents=<5 5 15>
      constraintSetup {angularLimitsMax=<15 2 20>, angularLimitsMin=<-15 -2 -20>, angularTarget=<1 0 0>}
      localJointOffset=<0 -0.25 0>
     }]
   }
   n1 = RigidBody(ComponentPose=Local To Component(LocalPose=Copy Pose From Mesh() {bUseAttachedParent=true})) {
    bOverrideWorldGravity=true
    overridePhysicsAsset=&/F_MED_Dieselpunk_02_Physics
    simulationSpace=BaseBoneSpace
    alphaBoolBlend {blendOption=Linear}
    baseBoneRef {boneName=spine_05}
    componentAppliedLinearAccClamp=<3000 3000 3000>
    componentLinearAccScale=<3 3 3>
    componentLinearVelScale=<3 3 3>
    overrideWorldGravity {z=-980}
   }
   Output Pose(Result=Component To Local(ComponentPose=@n1)) {layerGroup=None}
  ]}
 EventGraph {nodes [
   TryGetPawnOwner()
   Event BlueprintUpdateAnimation()
  ]}
]
variables []
